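; Floating platform lights actor for the Figment Cattus/Doggus event that switches a particle effect on and off.
; On BeginPlay it listens to the stateful Cattus platform lights event through the global event router and copies the payload's active flag into AreLightsActive.
; That variable's RepNotify activates or deactivates the particle component and prints a debug string.
; A Debug flag forces the lights on once listening starts.
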(event EventBeginPlay
  (bind (_asynctaskproxy _payload) (GameplayEventRouter|StartListeningToStatefulEvent (GameplayEventRouter|GetGlobalEventRouter) "/Figment_S09/Gameplay/EventRouter/ST_Figment_Cattus_PlatformLights_Event.ST_Figment_Cattus_PlatformLights_Event")
    (:then
      (if (Variables|Default|GetDebug)
        (Variables|Default|SetAreLightsActive true)))
    (:OnEventReceived
      _)
    (:OnSavedEventStateExists
      _))
  (bind _payload_active_1_cb711ec747b36eb39753a09a2a05d7b2 (Utilities|Struct|BreakSTFigmentCattusPlatformLightsEvent _payload))
  (Variables|Default|SetAreLightsActive _payload_active_1_cb711ec747b36eb39753a09a2a05d7b2))

(event Custom|ActivateLights
  (Development|PrintString "Cattus- Platform Lights - Activated")
  (Components|Activation|Activate (Variables|Default|GetParticleSystemComponent0) false))

(event Custom|DeactivateLights
  (Development|PrintString "Cattus- Platform Lights - Deactivated")
  (Components|Activation|Deactivate (Variables|Default|GetParticleSystemComponent0)))

(fn OnRep_AreLightsActive
  (if (Variables|Default|GetAreLightsActive)
    (CallFunction|ActivateLights)
    (else
      (CallFunction|DeactivateLights))))

(fn ConstructionScript)
