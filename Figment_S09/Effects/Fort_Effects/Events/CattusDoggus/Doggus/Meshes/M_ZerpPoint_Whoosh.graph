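# Unlit translucent mesh-particle material for the Zero Point whoosh effect in the CattusDoggus event.
# Layered panning smooth-noise samples, warped by each other and offset by reflection vector, build tinted, brightened streaks multiplied by particle color.
# Dynamic parameters drive top opacity, bottom fade, pan offset and a Ripple band that travels along UV1, distorting UVs, adding glow and scaling a world-position-offset height.
# Sphere masks, depth fade and axis-aligned Fresnel shape opacity.
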Material /Figment_S09/Effects/Fort_Effects/Events/CattusDoggus/Doggus/Meshes/M_ZerpPoint_Whoosh
graph {
 properties {
  bUsedWithMeshParticles=true
  blendMode=BLEND_Translucent
  pixelDepthOffsetMode=PDOM_Legacy
  previewMesh=&/SM_ZeroPoint_Whoosh
  shadingModel=MSM_Unlit
  usedShadingModels=Unlit
 }
 nodes [
  MaterialFunctionCall(Speed=Constant() {r=0.25}) {materialFunction=&DebugTimeSine}
  n1 = ScalarParameter() {defaultValue=5, parameterName=BrightnessMult}
  n2 = DynamicParameter() {defaultValue=<rgba 1 0 1 1>, paramNames [OpacityTop,FadeBottom,Ripple,PanAdd]}
  n3 = LinearInterpolate(A=Constant2Vector() {g=0.7, r=1}, B=Constant2Vector() {g=1.4, r=1}, Alpha=@n2.PanAdd)
  n4 = VectorParameter() {parameterName=Tile1, defaultValue=<rgba 6 2 0 0>}
  n5 = Multiply(A=@n3, B=AppendVector(A=@n4.R, B=@n4.R))
  n6 = TextureCoordinate() {coordinateIndex=1}
  n7 = OneMinus(None=ComponentMask(None=@n6) {g=true})
  n8 = Constant2Vector() {g=-0.5}
  n9 = Power(Base=ComponentMask(None=TextureCoordinate() {coordinateIndex=1}) {g=true}) {constExponent=0.3}
  n10 = ScalarParameter() {defaultValue=0.25, parameterName=SpherePanRadius}
  n11 = Multiply(A=@n10) {constB=0.25}
  n12 = Saturate(None=SphereMask(
    A=Saturate(None=@n9)
    B=LinearInterpolate(Alpha=@n2.Ripple) {constA=1, constB=-0.1}
    Radius=LinearInterpolate(A=@n11, B=@n10, Alpha=@n2.Ripple)
   ) {attenuationRadius=0.25, hardnessPercent=0})
  n13 = Add(A=@n6, B=Multiply(A=Constant2Vector() {g=0.1}, B=@n12))
  n14 = Add(A=Multiply(A=Power(Base=@n7) {constExponent=6}, B=@n8), B=@n13)
  n15 = Add(A=Multiply(A=Time(), B=ScalarParameter() {defaultValue=0.2, parameterName=PanSpeed}), B=@n2.PanAdd)
  n16 = VectorParameter() {parameterName=PanDirection1, defaultValue=<rgba -0.05 1.6 0 0>}
  n17 = TextureSample(UVs=Panner(Coordinate=Multiply(A=@n5, B=@n14), Time=@n15, Speed=AppendVector(A=@n16.R, B=@n16.R))) {texture=/Game/Effects/Fort_Effects/Textures/T_Smooth_tile}
  n18 = Constant3Vector() {constant=<rgba 1 0 0.4 0>}
  n19 = ComponentMask(None=ReflectionVectorWS()) {g=true, r=true}
  n20 = Power(Base=@n7) {constExponent=3}
  n21 = Add(A=Multiply(A=@n20, B=@n8) {constB=0.01}, B=@n13)
  n22 = VectorParameter() {parameterName=Tile2, defaultValue=<rgba 5 2 0 0>}
  n23 = Multiply(A=@n3, B=AppendVector(A=@n22.R, B=@n22.R))
  n24 = VectorParameter() {parameterName=PanDirection2, defaultValue=<rgba 0.04 0.75 0 0>}
  n25 = Add(
   A=Multiply(A=@n17.G, B=Constant2Vector() {g=-0.1, r=0.04}) {constB=0.1}
   B=Panner(Coordinate=Multiply(A=@n21, B=@n23), Time=@n15, Speed=AppendVector(A=@n24.R, B=@n24.R)) {speedX=0.1, speedY=0.04}
  )
  n26 = TextureSample(UVs=Add(A=Multiply(A=@n19) {constB=0.07}, B=@n25) {constB=0.01}) {texture=&T_Smooth_tile}
  n27 = Constant3Vector() {constant=<rgba 0 1 0.6 0>}
  n28 = TextureSample(UVs=@n25) {texture=&T_Smooth_tile}
  n29 = ComponentMask(None=TextureCoordinate() {coordinateIndex=1}) {g=true}
  n30 = Multiply(A=Add(A=Power(Base=@n12), B=Power(Base=@n29) {constExponent=4})) {constB=0.2}
  n31 = Add(A=@n30, B=LinearInterpolate(A=Multiply(A=@n18, B=@n26.G), B=Multiply(A=@n27, B=@n28.G)))
  n32 = LinearInterpolate(Alpha=Saturate(None=Add(A=@n17.G, B=@n20))) {constA=0.7, constB=3}
  n33 = Add(
   A=Multiply(A=VectorParameter() {parameterName=Color2, defaultValue=<rgba 0.1028 0.4793 0.6 0>}.RGB, B=@n17.G)
   B=Saturate(None=Power(Base=@n31, Exp=@n32))
  )
  Multiply(
   A=ParticleColor().RGB
   B=Add(A=VectorParameter() {parameterName=Background, defaultValue=<rgba 0.2872 0.2816 1 0>}.RGB, B=Multiply(A=@n1, B=@n33) {constB=5})
  )
  n34 = Add(A=@n33, B=Multiply(A=@n30) {constB=0.6})
  n35 = DynamicParameter() {defaultValue=<rgba 1 0 1 1>, paramNames [OpacityTop,FadeBottom,Ripple,PanAdd]}
  n36 = ComponentMask(None=TextureCoordinate() {coordinateIndex=1}) {g=true}
  n37 = Constant()
  n38 = Multiply(
   A=SphereMask(
    A=TextureCoordinate()
    B=Constant() {r=0.5}
    Radius=LinearInterpolate(Alpha=@n35.FadeBottom) {constA=0.5, constB=0}
    Hardness=ScalarParameter() {defaultValue=0.75, parameterName=Hardness}
   )
   B=Power(Base=SphereMask(A=@n36, B=@n37, Radius=@n35.OpacityTop) {hardnessPercent=0})
  )
  n39 = LinearInterpolate(Alpha=@n17.G) {constA=1, constB=0.5}
  n40 = ScalarParameter() {defaultValue=2000, parameterName=DepthFadeDist}
  n41 = Constant3Vector() {constant=<rgba 0 0 1 0>}
  n42 = ScalarParameter() {defaultValue=3, parameterName=Exp}
  Multiply(
   A=DepthFade(Opacity=Multiply(A=@n34, B=@n38), FadeDistance=Multiply(A=@n39, B=@n40))
   B=LinearInterpolate(
    B=MaterialFunctionCall(Axis=@n41, Exponent=@n42) {materialFunction=&AxisAlignedFresnel}.Fresnel
    Alpha=@n36
   ) {constA=1}
  )
  n43 = VectorParameter() {parameterName=WPOheight, defaultValue=<rgba 0 0 2500 0>}
  Reroute(None=Multiply(A=@n38, B=Multiply(A=@n12, B=@n43.RGB)))
 ]
}
